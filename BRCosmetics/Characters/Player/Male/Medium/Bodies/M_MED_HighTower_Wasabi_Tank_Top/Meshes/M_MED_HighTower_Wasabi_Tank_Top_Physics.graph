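# Physics asset for the male medium HighTower Wasabi tank top body, set up for a dangling necklace with dog tags.
# Capsules on upper spine, neck and clavicles approximate the torso around the necklace bones.
# Left and right four-link chains hang from tiny master bodies constrained to spine_05, with tight twist limits, and join at their ends into a closed loop.
# Two dog-tag bodies hang from the right chain's end, swinging mostly on one axis.
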PhysicsAsset /BRCosmetics/Characters/Player/Male/Medium/Bodies/M_MED_HighTower_Wasabi_Tank_Top/Meshes/M_MED_HighTower_Wasabi_Tank_Top_Physics
bodies {
 clavicle_l [{shapeType=Capsule, length=15, radius=5.75, center=<7.8202 0.0002 0>, rotation=<rot -90 0.0132 -0.0132>}]
 clavicle_r [{shapeType=Capsule, length=15, radius=5.75, center=<-7.8202 -0.0002 0>, rotation=<rot -90 -0.0198 -179.9802>}]
 dyn_necklaceDogtag_00 [{shapeType=Capsule, length=4.1147, radius=1.25, center=<2.6283 0.0001 1.3607>, rotation=<rot -90 0.6331 -0.6331>}]
 dyn_necklaceDogtag_01_l [{shapeType=Capsule, length=4.1147, radius=1.25, center=<2.6283 0.0002 0.9855>, rotation=<rot -90 0.6331 -0.6331>}]
 dyn_necklace_01_l [{shapeType=Capsule, length=5.5739, radius=1, center=<3.153 0 -0.0003>, rotation=<rot -90 -0.633 0.6329>}]
 dyn_necklace_01_l_MSTR [{shapeType=Capsule, length=0.01, radius=0.01, rotation=<rot 0 -90 0>}]
 dyn_necklace_01_r [{shapeType=Capsule, length=5.5739, radius=1, center=<3.153 0 -0.0003>, rotation=<rot -90 -0.633 0.6329>}]
 dyn_necklace_01_r_MSTR [{shapeType=Capsule, length=0.01, radius=0.01, rotation=<rot 0 -90 0>}]
 dyn_necklace_02_l [{shapeType=Capsule, length=5.5739, radius=1, center=<3.153 0 -0.0003>, rotation=<rot -90 -0.633 0.6329>}]
 dyn_necklace_02_r [{shapeType=Capsule, length=5.5739, radius=1, center=<3.153 0 -0.0003>, rotation=<rot -90 -0.633 0.6329>}]
 dyn_necklace_03_l [{shapeType=Capsule, length=5.5739, radius=1, center=<3.153 0 -0.0003>, rotation=<rot -90 -0.633 0.6329>}]
 dyn_necklace_03_r [{shapeType=Capsule, length=5.5739, radius=1, center=<3.153 0 -0.0003>, rotation=<rot -90 -0.633 0.6329>}]
 dyn_necklace_04_l [{shapeType=Capsule, length=5.5739, radius=1, center=<3.153 0.0001 0.3312>, rotation=<rot -90 -1.2658 1.2657>}]
 dyn_necklace_04_r [{shapeType=Capsule, length=5.5739, radius=1, center=<3.153 0 0.2863>, rotation=<rot -90 -1.8984 1.8983>}]
 dyn_necklace_bk_01_MSTR [{shapeType=Capsule, length=0.505, radius=0.505, rotation=<rot 0 -90 0>}]
 neck_01 [
  {shapeType=Capsule, length=12.8372, radius=6.3264, center=<4.4085 -0.6217 0>, rotation=<rot -90 0.0396 -0.0396>}
  {shapeType=Capsule, length=15.0872, radius=4.0764, center=<4.4084 4.2135 0.0001>, rotation=<rot -90 0 -10>}
 ]
 spine_04 [{shapeType=Capsule, length=17.5872, radius=13.5764, center=<5.1951 1.7905 0>}]
 spine_05 [{shapeType=Capsule, length=28.5872, radius=8.8264, center=<3.9328 -1.314 -0.0002>, rotation=<rot 0 -0.0071 0>}]
}
constraints [
 {bone1Name=spine_05, bone2Name=spine_04, swing1LimitDegrees=45, swing2LimitDegrees=45, twistLimitDegrees=45}
 {
  bone1Name=dyn_necklace_01_r_MSTR
  bone2Name=spine_05
  swing1LimitDegrees=45
  swing1Motion=Free
  swing2LimitDegrees=45
  swing2Motion=Free
  twistLimitDegrees=45
  twistMotion=Free
 }
 {
  bone1Name=dyn_necklace_01_l_MSTR
  bone2Name=spine_05
  swing1LimitDegrees=45
  swing1Motion=Free
  swing2LimitDegrees=45
  swing2Motion=Free
  twistLimitDegrees=45
  twistMotion=Free
 }
 {bone1Name=dyn_necklace_01_l, bone2Name=dyn_necklace_01_l_MSTR, swing1LimitDegrees=45, swing2LimitDegrees=45, twistLimitDegrees=5}
 {bone1Name=dyn_necklace_02_l, bone2Name=dyn_necklace_01_l, swing1LimitDegrees=45, swing2LimitDegrees=45, twistLimitDegrees=5}
 {bone1Name=dyn_necklace_03_l, bone2Name=dyn_necklace_02_l, swing1LimitDegrees=45, swing2LimitDegrees=45, twistLimitDegrees=5}
 {bone1Name=dyn_necklace_04_l, bone2Name=dyn_necklace_03_l, swing1LimitDegrees=45, swing2LimitDegrees=45, twistLimitDegrees=5}
 {bone1Name=dyn_necklace_01_r, bone2Name=dyn_necklace_01_r_MSTR, swing1LimitDegrees=45, swing2LimitDegrees=45, twistLimitDegrees=5}
 {bone1Name=dyn_necklace_02_r, bone2Name=dyn_necklace_01_r, swing1LimitDegrees=45, swing2LimitDegrees=45, twistLimitDegrees=5}
 {bone1Name=dyn_necklace_03_r, bone2Name=dyn_necklace_02_r, swing1LimitDegrees=45, swing2LimitDegrees=45, twistLimitDegrees=5}
 {bone1Name=dyn_necklace_04_r, bone2Name=dyn_necklace_03_r, swing1LimitDegrees=45, swing2LimitDegrees=45, twistLimitDegrees=5}
 {
  bone1Name=dyn_necklace_04_l
  bone2Name=dyn_necklace_04_r
  swing1LimitDegrees=45
  swing1Motion=Free
  swing2LimitDegrees=45
  swing2Motion=Free
  twistLimitDegrees=45
  twistMotion=Free
 }
 {
  bone1Name=neck_01
  bone2Name=spine_05
  swing1LimitDegrees=45
  swing1Motion=Free
  swing2LimitDegrees=45
  swing2Motion=Free
  twistLimitDegrees=45
  twistMotion=Free
 }
 {
  bone1Name=dyn_necklace_bk_01_MSTR
  bone2Name=neck_01
  swing1LimitDegrees=45
  swing1Motion=Free
  swing2LimitDegrees=45
  swing2Motion=Free
  twistLimitDegrees=45
  twistMotion=Free
 }
 {
  bone1Name=clavicle_l
  bone2Name=spine_05
  swing1LimitDegrees=45
  swing1Motion=Free
  swing2LimitDegrees=45
  swing2Motion=Free
  twistLimitDegrees=45
  twistMotion=Free
 }
 {
  bone1Name=clavicle_r
  bone2Name=spine_05
  swing1LimitDegrees=45
  swing1Motion=Free
  swing2LimitDegrees=45
  swing2Motion=Free
  twistLimitDegrees=45
  twistMotion=Free
 }
 {bone1Name=dyn_necklaceDogtag_00, bone2Name=dyn_necklace_04_r, swing1LimitDegrees=80, swing2LimitDegrees=5, twistLimitDegrees=5}
 {bone1Name=dyn_necklaceDogtag_01_l, bone2Name=dyn_necklaceDogtag_00, swing1LimitDegrees=80, swing2LimitDegrees=5, twistLimitDegrees=5}
]
