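# Physics asset that drives secondary hair motion for the male medium HighTower Wasabi Tank Top cosmetic character part.
# A head capsule anchors three hair groups: a two-link front strand and single left and right top pieces.
# Each group hangs from a tiny MSTR capsule body whose angular motion is free relative to the head.
# Tight swing limits constrain the hair capsules beneath those masters, with twist locked on the top pieces.
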PhysicsAsset /BRCosmetics/Characters/Player/Male/Medium/Bodies/M_MED_HighTower_Wasabi_Tank_Top/Meshes/Parts/M_MED_HighTowerWasabiTankTop_HairPhysics
bodies {
 dyn_hairFront_01 [{shapeType=Capsule, length=2.8889, radius=0.7536, center=<1.0893 0 0>, rotation=<rot 90 -2.5305 -2.5305>}]
 dyn_hairFront_01_MSTR [{shapeType=Capsule, length=0.01, radius=0.01}]
 dyn_hairFront_02 [{shapeType=Capsule, length=2.8889, radius=0.7536, center=<1.4278 0 0>, rotation=<rot 90 -2.5305 -2.5305>}]
 dyn_hairtop_01_l [{shapeType=Capsule, length=4.6811, radius=4.0676, center=<0.909 0 -0.0001>, rotation=<rot 90 -0.0033 179.9968>}]
 dyn_hairtop_01_l_MSTR [{shapeType=Capsule, length=0.01, radius=0.01}]
 dyn_hairtop_01_r [{shapeType=Capsule, length=4.6811, radius=4.0676, center=<0.909 0 -0.0001>, rotation=<rot 90 -0.0033 179.9968>}]
 dyn_hairtop_01_r_MSTR [{shapeType=Capsule, length=0.01, radius=0.01}]
 head [{shapeType=Capsule, length=6.502, radius=8.9022, center=<4.5467 1.1528 -0.0001>, rotation=<rot 90 0.0395 0.0394>}]
}
constraints [
 {bone1Name=dyn_hairFront_01, bone2Name=dyn_hairFront_01_MSTR, swing1LimitDegrees=5, swing2LimitDegrees=5, twistLimitDegrees=5}
 {bone1Name=dyn_hairFront_02, bone2Name=dyn_hairFront_01, swing1LimitDegrees=15, swing2LimitDegrees=15, twistLimitDegrees=5}
 {
  bone1Name=dyn_hairtop_01_r
  bone2Name=dyn_hairtop_01_r_MSTR
  swing1LimitDegrees=15
  swing2LimitDegrees=15
  twistLimitDegrees=5
  twistMotion=Locked
 }
 {
  bone1Name=dyn_hairFront_01_MSTR
  bone2Name=head
  swing1LimitDegrees=45
  swing1Motion=Free
  swing2LimitDegrees=45
  swing2Motion=Free
  twistLimitDegrees=45
  twistMotion=Free
 }
 {
  bone1Name=dyn_hairtop_01_r_MSTR
  bone2Name=head
  swing1LimitDegrees=45
  swing1Motion=Free
  swing2LimitDegrees=45
  swing2Motion=Free
  twistLimitDegrees=45
  twistMotion=Free
 }
 {
  bone1Name=dyn_hairtop_01_l_MSTR
  bone2Name=head
  swing1LimitDegrees=45
  swing1Motion=Free
  swing2LimitDegrees=45
  swing2Motion=Free
  twistLimitDegrees=45
  twistMotion=Free
 }
 {
  bone1Name=dyn_hairtop_01_l
  bone2Name=dyn_hairtop_01_l_MSTR
  swing1LimitDegrees=15
  swing2LimitDegrees=15
  twistLimitDegrees=5
  twistMotion=Locked
 }
]
