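# Anim Blueprint for the HighTower Wasabi tank-top body part, adding procedural muscle deformation and hair physics over the pose copied from the attached parent mesh.
# Constraints, Look Ats and additive bone transforms drive pecs, biceps, scapulae, traps, lats and right tricep.
# The event graph derives stretch alphas from bone distances each update, plus several flex and joint alphas the AnimGraph leaves unused.
# A neck-relative RigidBody node takes skydiving-dependent wind as gravity and blend weight.
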AnimBlueprint /BRCosmetics/Characters/Player/Male/Medium/Bodies/M_MED_HighTower_Wasabi_Tank_Top/Meshes/Parts/M_MED_HighTower_Wasabi_Tank_Top_AnimBP
target_skeleton=&/M_MED_HighTower_Wasabi_Tank_Top_Skeleton
graphs [
 AnimGraph {nodes [
   n1 = Get RpecStretchAlpha() -> (RpecStretchAlpha:double)
   "Transform (Modify) Bone - Bone: bicep_02_r"(Alpha=@n1, Translation="0,0,0", Rotation="0,0,0", Scale="0.15,0.15,0.15") {
    rotationMode=BMM_Additive
    rotationSpace=BCS_BoneSpace
    scaleMode=BMM_Additive
    scaleSpace=BCS_BoneSpace
    alphaBoolBlend {blendOption=Linear}
    boneToModify {boneName=bicep_02_r}
    scale=<0.15 0.15 0.15>
   }
   n2 = Get LpecStretchAlpha() -> (LpecStretchAlpha:double)
   "Transform (Modify) Bone - Bone: bicep_02_l"(Alpha=@n2, Translation="0,0,0", Rotation="0,0,0", Scale="0.15,0.15,0.15") {
    rotationMode=BMM_Additive
    rotationSpace=BCS_BoneSpace
    scaleMode=BMM_Additive
    scaleSpace=BCS_BoneSpace
    alphaBoolBlend {blendOption=Linear}
    boneToModify {boneName=bicep_02_l}
    scale=<0.15 0.15 0.15>
   }
   n3 = "Constraint - pectoralis_02_r_aim (PPPPPPP)"!(
    ComponentPose=Local To Component(LocalPose=Copy Pose From Mesh() {bUseAttachedParent=true})
    Alpha=1
    ConstraintWeights_0=0.0036
    ConstraintWeights_1:float=0.567
    ConstraintWeights_2:float=0.0351
    ConstraintWeights_3:float=0.0418
    ConstraintWeights_4:float=0.0444
    ConstraintWeights_5:float=0.0255
    ConstraintWeights_6:float=0.2826
   ) -> (Pose) {
    alphaBoolBlend {blendOption=Linear}
    boneToModify {boneName=pectoralis_02_r_aim}
    constraintSetup [
     {targetBone {boneName=upperarm_r}}
     {targetBone {boneName=upperarm_twist_01_r}}
     {targetBone {boneName=pectoralis_02_r}}
     {targetBone {boneName=spine_03}}
     {targetBone {boneName=spine_04}}
     {targetBone {boneName=spine_05}}
     {targetBone {boneName=clavicle_r}}
    ]
    constraintWeights [1,1,1,1,1,1,1]
   }
   n4 = "Constraint - pectoralis_02_l_aim (PPPPPPP)"!(
    ComponentPose=@n3
    Alpha=1
    ConstraintWeights_0=0.2826
    ConstraintWeights_1:float=0.0351
    ConstraintWeights_2:float=0.0418
    ConstraintWeights_3:float=0.0444
    ConstraintWeights_4:float=0.0255
    ConstraintWeights_5:float=0.567
    ConstraintWeights_6:float=0.0036
   ) -> (Pose) {
    alphaBoolBlend {blendOption=Linear}
    boneToModify {boneName=pectoralis_02_l_aim}
    constraintSetup [
     {targetBone {boneName=clavicle_l}}
     {targetBone {boneName=pectoralis_02_l}}
     {targetBone {boneName=spine_03}}
     {targetBone {boneName=spine_04}}
     {targetBone {boneName=spine_05}}
     {targetBone {boneName=upperarm_twist_01_l}}
     {targetBone {boneName=upperarm_l}}
    ]
    constraintWeights [1,1,1,1,1,1,1]
   }
   n5 = Look At - Bone: pectoralis_01_l!(ComponentPose=@n4, Alpha=1, LookAtLocation:Vector="0,0,0") -> (Pose) {
    alphaBoolBlend {blendOption=Linear}
    boneToModify {boneName=pectoralis_01_l}
    lookAtLocation {x=0}
    lookAtTarget {boneReference {boneName=pectoralis_02_l_aim}}
    lookAt_Axis {axis=<1 0>}
    lookUp_Axis {axis {x=0, z=1}}
   }
   n6 = Look At - Bone: pectoralis_01_r!(ComponentPose=@n5, Alpha=1, LookAtLocation:Vector="0,0,0") -> (Pose) {
    alphaBoolBlend {blendOption=Linear}
    boneToModify {boneName=pectoralis_01_r}
    lookAtLocation {x=0}
    lookAtTarget {boneReference {boneName=pectoralis_02_r_aim}}
    lookAt_Axis {axis=<1 0>}
    lookUp_Axis {axis {x=0, z=1}}
   }
   n7 = "Constraint - scapula_base_l (P)"(ComponentPose=@n6, Alpha=1, ConstraintWeights_0=1) {
    alphaBoolBlend {blendOption=Linear}
    boneToModify {boneName=scapula_base_l}
    constraintSetup [{targetBone {boneName=clavicle_l}}]
    constraintWeights [1]
   }
   n8 = Look At - Bone: scapula_00_l(ComponentPose=@n7, Alpha=1) {
    alphaBoolBlend {blendOption=Linear}
    boneToModify {boneName=scapula_00_l}
    lookAtLocation {x=0}
    lookAtTarget {boneReference {boneName=scapula_01_l_aim}}
    lookAt_Axis {axis=<1 0>}
    lookUp_Axis {axis {x=0, z=1}}
   }
   n9 = Look At - Bone: lattissimus_00_l(ComponentPose=@n8, Alpha=1) {
    alphaBoolBlend {blendOption=Linear}
    boneToModify {boneName=lattissimus_00_l}
    lookAtLocation {x=0}
    lookAtTarget {boneReference {boneName=upperarm_l}}
    lookAt_Axis {axis=<1 0>}
    lookUp_Axis {axis {x=0, z=1}}
   }
   n10 = "Constraint - scapula_base_r (P)"(ComponentPose=@n9, Alpha=1, ConstraintWeights_0=1) {
    alphaBoolBlend {blendOption=Linear}
    boneToModify {boneName=scapula_base_r}
    constraintSetup [{targetBone {boneName=clavicle_r}}]
    constraintWeights [1]
   }
   n11 = Look At - Bone: scapula_00_r(ComponentPose=@n10, Alpha=1) {
    alphaBoolBlend {blendOption=Linear}
    boneToModify {boneName=scapula_00_r}
    lookAtLocation {x=0}
    lookAtTarget {boneReference {boneName=scapula_01_r_aim}}
    lookAt_Axis {axis=<1 0>}
    lookUp_Axis {axis {x=0, z=1}}
   }
   n12 = Look At - Bone: lattissimus_00_r(ComponentPose=@n11, Alpha=1) {
    alphaBoolBlend {blendOption=Linear}
    boneToModify {boneName=lattissimus_00_r}
    lookAtLocation {x=0}
    lookAtTarget {boneReference {boneName=upperarm_r}}
    lookAt_Axis {axis=<1 0>}
    lookUp_Axis {axis {x=0, z=1}}
   }
   n13 = "Transform (Modify) Bone - Bone: bicep_tp_00_r"(ComponentPose=@n12, Alpha=1, Translation="0,0,0", Rotation="0,0,0", Scale="1,1,1") {
    rotationMode=BMM_Additive
    rotationSpace=BCS_BoneSpace
    scaleMode=BMM_Additive
    scaleSpace=BCS_BoneSpace
    alphaBoolBlend {blendOption=Linear}
    boneToModify {boneName=bicep_tp_00_r}
   }
   n14 = "Transform (Modify) Bone - Bone: bicep_bt_00_r"(ComponentPose=@n13, Alpha=1, Translation="0,0,0", Rotation="0.000,0.000,-20", Scale="1,1,1") {
    rotationMode=BMM_Additive
    rotationSpace=BCS_BoneSpace
    alphaBoolBlend {blendOption=Linear}
    boneToModify {boneName=bicep_bt_00_r}
    rotation {roll=-20}
   }
   n15 = "Transform (Modify) Bone - Bone: bicep_01_r"(ComponentPose=@n14, Alpha=@n1, Translation="0,0,0", Rotation="0,0,0", Scale="0.15,0.15,0.15") {
    rotationMode=BMM_Additive
    rotationSpace=BCS_BoneSpace
    scaleMode=BMM_Additive
    scaleSpace=BCS_BoneSpace
    alphaBoolBlend {blendOption=Linear}
    boneToModify {boneName=bicep_01_r}
    scale=<0.15 0.15 0.15>
   }
   n16 = "Transform (Modify) Bone - Bone: pectoralis_02_r"(ComponentPose=@n15, Alpha=@n1, Translation="-8.560,0.000,0", Rotation="0,0,0", Scale="1,1,1") {
    rotationMode=BMM_Additive
    rotationSpace=BCS_BoneSpace
    scaleMode=BMM_Additive
    scaleSpace=BCS_BoneSpace
    translationMode=BMM_Additive
    translationSpace=BCS_BoneSpace
    alphaBoolBlend {blendOption=Linear}
    boneToModify {boneName=pectoralis_02_r}
    translation {x=-8.56}
   }
   n17 = "Transform (Modify) Bone - Bone: bicep_01_l"(ComponentPose=@n16, Alpha=@n2, Translation="0,0,0", Rotation="0,0,0", Scale="0.15,0.15,0.15") {
    rotationMode=BMM_Additive
    rotationSpace=BCS_BoneSpace
    scaleMode=BMM_Additive
    scaleSpace=BCS_BoneSpace
    alphaBoolBlend {blendOption=Linear}
    boneToModify {boneName=bicep_01_l}
    scale=<0.15 0.15 0.15>
   }
   n18 = "Transform (Modify) Bone - Bone: pectoralis_02_l"(ComponentPose=@n17, Alpha=@n2, Translation="-8.560,0.000,-0.000", Rotation="0,0,0", Scale="1,1,1") {
    rotationMode=BMM_Additive
    rotationSpace=BCS_BoneSpace
    scaleMode=BMM_Additive
    scaleSpace=BCS_BoneSpace
    translationMode=BMM_Additive
    translationSpace=BCS_BoneSpace
    alphaBoolBlend {blendOption=Linear}
    boneToModify {boneName=pectoralis_02_l}
    translation {x=-8.5597, z=0}
   }
   n19 = Look At - Bone: dyn_trapezius_01_r(ComponentPose=@n18, Alpha=1) {
    alphaBoolBlend {blendOption=Linear}
    boneToModify {boneName=dyn_trapezius_01_r}
    lookAtLocation {x=0}
    lookAtTarget {boneReference {boneName=dyn_trapezius_02_r_aim}}
    lookAt_Axis {axis=<1 0>}
    lookUp_Axis {axis {x=0, z=1}}
   }
   n20 = Look At - Bone: dyn_trapezius_01_l(ComponentPose=@n19, Alpha=1) {
    alphaBoolBlend {blendOption=Linear}
    boneToModify {boneName=dyn_trapezius_01_l}
    lookAtLocation {x=0}
    lookAtTarget {boneReference {boneName=dyn_trapezius_02_l_aim}}
    lookAt_Axis {axis=<1 0>}
    lookUp_Axis {axis {x=0, z=1}}
   }
   n21 = "Constraint - dyn_trapezius_02_l (T)"(ComponentPose=@n20, Alpha=Get trapsStretchAlpha_l() -> (trapsStretchAlpha_l:double), ConstraintWeights_0=1) {
    alphaBoolBlend {blendOption=Linear}
    boneToModify {boneName=dyn_trapezius_02_l}
    constraintSetup [{offsetOption=None, transformType=Translation, targetBone {boneName=dyn_trapezius_02_l_aim}}]
    constraintWeights [1]
   }
   n22 = "Constraint - dyn_trapezius_02_r (T)"(ComponentPose=@n21, Alpha=Get trapsStretchAlpha_r() -> (trapsStretchAlpha_r:double), ConstraintWeights_0=1) {
    alphaBoolBlend {blendOption=Linear}
    boneToModify {boneName=dyn_trapezius_02_r}
    constraintSetup [{offsetOption=None, transformType=Translation, targetBone {boneName=dyn_trapezius_02_r_aim}}]
    constraintWeights [1]
   }
   n23 = "Constraint - dyn_lat_measure_r (PPPPPPPP)"!(
    ComponentPose=@n22
    Alpha=1
    ConstraintWeights_0=0.1592
    ConstraintWeights_1:float=0.0728
    ConstraintWeights_2:float=0.0847
    ConstraintWeights_3:float=0.2652
    ConstraintWeights_4:float=0.2361
    ConstraintWeights_5:float=0.1157
    ConstraintWeights_6:float=0
    ConstraintWeights_7:float=0.0664
   ) -> (Pose) {
    alphaBoolBlend {blendOption=Linear}
    boneToModify {boneName=dyn_lat_measure_r}
    constraintSetup [
     {targetBone {boneName=lattissimus_01_r}}
     {targetBone {boneName=scapula_01_r}}
     {targetBone {boneName=spine_02}}
     {targetBone {boneName=spine_03}}
     {targetBone {boneName=spine_04}}
     {targetBone {boneName=spine_05}}
     {targetBone {boneName=upperarm_twist_01_l}}
     {targetBone {boneName=clavicle_r}}
    ]
    constraintWeights [1,1,1,1,1,1,1,1]
   }
   n24 = "Constraint - dyn_tricep_measure_r (PPPPPPPP)"!(
    ComponentPose=@n23
    Alpha=1
    ConstraintWeights_0=0.192
    ConstraintWeights_1:float=0.2995
    ConstraintWeights_2:float=0.1418
    ConstraintWeights_3:float=0.0017
    ConstraintWeights_4:float=0.0952
    ConstraintWeights_5:float=0.0522
    ConstraintWeights_6:float=0.1161
    ConstraintWeights_7:float=0.1013
   ) -> (Pose) {
    alphaBoolBlend {blendOption=Linear}
    boneToModify {boneName=dyn_tricep_measure_r}
    constraintSetup [
     {targetBone {boneName=tricep_01_r}}
     {targetBone {boneName=upperarm_r}}
     {targetBone {boneName=upperarm_twist_01_r}}
     {targetBone {boneName=lattissimus_01_r}}
     {targetBone {boneName=spine_03}}
     {targetBone {boneName=spine_04}}
     {targetBone {boneName=tricep_02_r}}
     {targetBone {boneName=clavicle_r}}
    ]
    constraintWeights [1,1,1,1,1,1,1,1]
   }
   n25 = Get RlatStretchAlpha() -> (RlatStretchAlpha:double)
   n26 = "Transform (Modify) Bone - Bone: tricep_01_r"(ComponentPose=@n24, Alpha=@n25, Translation="-0.000,-5.788,-0.000", Rotation="0,0,0", Scale="1.000,0.8,1.000") {
    rotationMode=BMM_Additive
    rotationSpace=BCS_BoneSpace
    scaleMode=BMM_Additive
    scaleSpace=BCS_BoneSpace
    translationMode=BMM_Additive
    translationSpace=BCS_BoneSpace
    alphaBoolBlend {blendOption=Linear}
    boneToModify {boneName=tricep_01_r}
    scale {y=0.8}
    translation {y=-5.788}
   }
   n27 = "Transform (Modify) Bone - Bone: tricep_02_r"(ComponentPose=@n26, Alpha=@n25, Translation="-0.000,-1.148,0.000", Rotation="0,0,0", Scale="1.000,0.8,1.000") {
    rotationMode=BMM_Additive
    rotationSpace=BCS_BoneSpace
    scaleMode=BMM_Additive
    scaleSpace=BCS_BoneSpace
    translationMode=BMM_Additive
    translationSpace=BCS_BoneSpace
    alphaBoolBlend {blendOption=Linear}
    boneToModify {boneName=tricep_02_r}
    scale {y=0.8}
    translation {y=-1.148}
   }
   n28 = "Transform (Modify) Bone - Bone: lattissimus_01_r"(ComponentPose=@n27, Alpha=@n25, Translation="0.000,-7.314,-0.000", Rotation="0,0,0", Scale="1.000,0.000,1.000") {
    rotationMode=BMM_Additive
    rotationSpace=BCS_BoneSpace
    scaleMode=BMM_Additive
    scaleSpace=BCS_BoneSpace
    translationMode=BMM_Additive
    translationSpace=BCS_BoneSpace
    alphaBoolBlend {blendOption=Linear}
    boneToModify {boneName=lattissimus_01_r}
    scale {y=0}
    translation {y=-7.314}
   }
   n29 = "Constraint - lattissimus_01_l (T)"(ComponentPose=@n28, Alpha=1, ConstraintWeights_0=1) {
    alphaBoolBlend {blendOption=Linear}
    boneToModify {boneName=lattissimus_01_l}
    constraintSetup [{transformType=Translation, targetBone {boneName=upperarm_l}}]
    constraintWeights [1]
   }
   n30 = "Constraint - lattissimus_01_r (T)"(ComponentPose=@n29, Alpha=1, ConstraintWeights_0=1) {
    alphaBoolBlend {blendOption=Linear}
    boneToModify {boneName=lattissimus_01_r}
    constraintSetup [{transformType=Translation, targetBone {boneName=upperarm_r}}]
    constraintWeights [1]
   }
   n31 = RigidBody!(
    ComponentPose=@n30
    Alpha=Get isSkydivingAlpha() -> (isSkydivingAlpha:double)
    OverrideWorldGravity:Vector=Get windVector() -> (windVector:Vector)
    ExternalForce
   ) -> (Pose) {
    bClampLinearTranslationLimitToRefPose=true
    bOverrideWorldGravity=true
    overridePhysicsAsset=&/M_MED_HighTowerWasabiTankTop_HairPhysics
    simulationSpace=BaseBoneSpace
    alphaBoolBlend {blendOption=Linear}
    baseBoneRef {boneName=neck_01}
    componentAppliedLinearAccClamp=<3000 3000 3000>
    componentLinearAccScale=<0.3 0.3 0.3>
    componentLinearVelScale=<0.3 0.3 0.35>
    overrideWorldGravity {z=-980}
   }
   Output Pose(Result=Component To Local(ComponentPose=@n31)) {layerGroup=None}
   Get RpecAimPosAlpha() -> (RpecAimPosAlpha:double)
  ]}
 EventGraph {nodes [
   Event BlueprintUpdateAnimation()
   > n1 = Sequence!(execute) -> (then_0, then_1, then_2:Exec, then_3:Exec, then_4:Exec, then_5:Exec)
   n2 = Break Fort Anim Input Skydiving External Force(FortAnimInput_SkydivingExternalForce=Get SkydivingExternalForceInput()) -> (
    bUseSkydivingVectorForce=~
    SkydivingVectorForce=~
    HeadToPelvisDirection=~
    FloatingMultiplier=~
    FloatingAdditive=~
    DivingMultiplier=~
    DivingAdditive=~
    ParachutingMultiplier=~
    ParachutingAdditive=~
   )
   > Set windVector(windVector:Vector=SelectVector(A=vector * float(A=@n2.SkydivingVectorForce, B=0.15), B="0, 0,-980", bPickA=Get bIsSkyDiving())) -> (Output_Get:Vector)
   > Set isSkydivingAlpha(isSkydivingAlpha:double=SelectFloat(A=0.5, B=1, bPickA=Get bIsSkyDiving())) -> (Output_Get:double)
   n3 = GetPartSkeletalMeshComponent()
   n4 = GetAlphaFromReferenceTransformDelta(
    SkeletalMeshComponent=@n3
    BoneName=lowerarm_r
    BaseName=upperarm_r
    DeltaComponentType=QuaternionTwist
    InputRange=(X=0,Y=-90)
    OutputRange=(X=0,Y=0.7)
    TwistAxis="0,0,1"
   )
   Set RflexAlpha(
    execute=@n1.then_1
    RflexAlpha:double=SelectFloat(A=0.25, B=MapRangeClamped(Value=@n4, InRangeA=0, InRangeB=1, OutRangeA=0.2, OutRangeB=0.7), bPickA=Get bIsTargeting())
   ) -> (Output_Get:double)
   n5 = GetAlphaFromReferenceTransformDelta(
    SkeletalMeshComponent=@n3
    BoneName=upperarm_r
    BaseName=clavicle_r
    DeltaComponentType=QuaternionTwist
    InputRange=(X=0,Y=-128)
    TwistAxis="0,0,1"
   )
   > Set RshoulderAlpha(RshoulderAlpha:double=@n5) -> (Output_Get:double)
   > Set RaimPecPos(RaimPecPos:Vector=MakeVector(Y=0, Z=MapRangeClamped(Value=@n5, InRangeA=0, InRangeB=1, OutRangeB=2))) -> (Output_Get:Vector)
   > Set RpecAimPosAlpha(RpecAimPosAlpha:double=@n5) -> (Output_Get:double)
   n6 = GetAlphaFromReferenceTransformDelta(
    SkeletalMeshComponent=@n3
    BoneName=upperarm_l
    BaseName=clavicle_l
    DeltaComponentType=QuaternionTwist
    InputRange=(X=-20,Y=-75)
    TwistAxis="0,0,1"
   )
   > Set LshoulderAlpha(LshoulderAlpha:double=@n6) -> (Output_Get:double)
   > Set LaimPecPos(LaimPecPos:Vector=MakeVector(Y=0, Z=MapRangeClamped(Value=@n6, InRangeA=0, InRangeB=1, OutRangeB=10))) -> (Output_Get:Vector)
   > Set RlegAlpha(RlegAlpha:double=SelectFloat(
     A=1
     B=GetAlphaFromReferenceTransformDelta(SkeletalMeshComponent=@n3, BoneName=thigh_r, BaseName=pelvis, DeltaComponentType=QuaternionTwist, InputRange=(X=0,Y=100), TwistAxis="0,0,1")
     bPickA=Get bIsCrouching()
    )) -> (Output_Get:double)
   n7 = GetPartSkeletalMeshComponent()
   Set LbicepFlex(
    execute=@n1.then_2
    LbicepFlex:double=GetAlphaFromReferenceTransformDelta(
     SkeletalMeshComponent=@n7
     BoneName=lowerarm_l
     BaseName=upperarm_l
     DeltaComponentType=QuaternionTwist
     InputRange=(X=-20,Y=-70)
     TwistAxis="0,0,1"
    )
   ) -> (Output_Get:double)
   > Set RbicepFlex(RbicepFlex:double=GetAlphaFromReferenceTransformDelta(
     SkeletalMeshComponent=@n7
     BoneName=lowerarm_r
     BaseName=upperarm_r
     DeltaComponentType=QuaternionTwist
     InputRange=(X=-20,Y=-70)
     TwistAxis="0,0,1"
    )) -> (Output_Get:double)
   > n8 = TransformFromBoneSpace!(
    execute
    self:Skinned Mesh Component Object Reference=@n7
    BoneName:Name=upperarm_l
    InPosition:Vector="0, 0, 0"
    InRotation:Rotator="0, 0, 0"
   ) -> (then, OutPosition:Vector, OutRotation:Rotator)
   > n9 = TransformFromBoneSpace(self=@n7, BoneName=lowerarm_twist_02_l, InPosition="0, 0, 0", InRotation="0, 0, 0")
   > Set LarmDistanceBICEPAlpha(LarmDistanceBICEPAlpha:double=MapRangeClamped(Value="Distance (Vector)"!(V1:Vector=@n8, V2:Vector=@n9) -> (ReturnValue:double), InRangeA=27, InRangeB=34, OutRangeA=1, OutRangeB=0)) -> (Output_Get:double)
   > n10 = TransformFromBoneSpace(self=@n7, BoneName=upperarm_r, InPosition="0, 0, 0", InRotation="0, 0, 0")
   > n11 = TransformFromBoneSpace(self=@n7, BoneName=lowerarm_twist_02_r, InPosition="0, 0, 0", InRotation="0, 0, 0")
   > Set RarmDistanceBICEPAlpha(RarmDistanceBICEPAlpha:double=MapRangeClamped(Value="Distance (Vector)"(V1=@n10, V2=@n11), InRangeA=27, InRangeB=34, OutRangeA=1, OutRangeB=0)) -> (Output_Get:double)
   > n12 = TransformFromBoneSpace(self=@n7, BoneName=upperarm_r, InPosition="0, 0, 0", InRotation="0, 0, 0")
   > n13 = TransformFromBoneSpace(self=@n7, BoneName=hand_r, InPosition="0, 0, 0", InRotation="0, 0, 0")
   > Set RarmDistanceAlpha(RarmDistanceAlpha:double=MapRangeClamped(Value="Distance (Vector)"(V1=@n12, V2=@n13), InRangeA=40, InRangeB=54, OutRangeA=0.7, OutRangeB=0)) -> (Output_Get:double)
   > n14 = TransformFromBoneSpace(self=@n7, BoneName=pectoralis_01_r, InPosition="0, 0, 0", InRotation="0, 0, 0")
   > n15 = TransformFromBoneSpace(self=@n7, BoneName=bicep_01_r, InPosition="0, 0, 0", InRotation="0, 0, 0")
   > Set RpecStretchAlpha(RpecStretchAlpha:double=MapRangeClamped(Value="Distance (Vector)"(V1=@n14, V2=@n15), InRangeA=23, InRangeB=12, OutRangeB=0.5)) -> (Output_Get:double)
   > n16 = TransformFromBoneSpace(self=@n7, BoneName=pectoralis_01_l, InPosition="0, 0, 0", InRotation="0, 0, 0")
   > n17 = TransformFromBoneSpace(self=@n7, BoneName=bicep_01_l, InPosition="0, 0, 0", InRotation="0, 0, 0")
   > Set LpecStretchAlpha(LpecStretchAlpha:double=MapRangeClamped(Value="Distance (Vector)"(V1=@n16, V2=@n17), InRangeA=23, InRangeB=12, OutRangeB=0.5)) -> (Output_Get:double)
   > n18 = TransformFromBoneSpace(self=@n7, BoneName=dyn_tricep_measure_r, InPosition="0, 0, 0", InRotation="0, 0, 0")
   > n19 = TransformFromBoneSpace(self=@n7, BoneName=dyn_lat_measure_r, InPosition="0, 0, 0", InRotation="0, 0, 0")
   > Set RlatStretchAlpha(RlatStretchAlpha:double=MapRangeClamped(Value="Distance (Vector)"(V1=@n18, V2=@n19), InRangeA=10, InRangeB=5)) -> (Output_Get:double)
   > n20 = TransformFromBoneSpace(self=@n7, BoneName=dyn_trapezius_01_l, InPosition="0, 0, 0", InRotation="0, 0, 0")
   > n21 = TransformFromBoneSpace(self=@n7, BoneName=dyn_trapezius_02_l_aim, InPosition="0, 0, 0", InRotation="0, 0, 0")
   > Set trapsStretchAlpha_l(trapsStretchAlpha_l:double=MapRangeClamped(Value="Distance (Vector)"(V1=@n20, V2=@n21), InRangeA=14, InRangeB=10)) -> (Output_Get:double)
   > n22 = TransformFromBoneSpace(self=@n7, BoneName=dyn_trapezius_01_r, InPosition="0, 0, 0", InRotation="0, 0, 0")
   > n23 = TransformFromBoneSpace(self=@n7, BoneName=dyn_trapezius_02_r_aim, InPosition="0, 0, 0", InRotation="0, 0, 0")
   > Set trapsStretchAlpha_r(trapsStretchAlpha_r:double=MapRangeClamped(Value="Distance (Vector)"(V1=@n22, V2=@n23), InRangeA=14, InRangeB=10)) -> (Output_Get:double)
   Set kneeAlpha_r(
    execute=@n1.then_4
    kneeAlpha_r:double=GetAlphaFromReferenceTransformDelta(SkeletalMeshComponent=@n3, BoneName=calf_r, BaseName=thigh_r, DeltaComponentType=QuaternionTwist, InputRange=(X=0,Y=-110), TwistAxis="0,0,1")
   ) -> (Output_Get:double)
   > Set kneeAlpha_l(kneeAlpha_l:double=GetAlphaFromReferenceTransformDelta(SkeletalMeshComponent=@n3, BoneName=calf_l, BaseName=thigh_l, DeltaComponentType=QuaternionTwist, InputRange=(X=0,Y=-110), TwistAxis="0,0,1")) -> (Output_Get:double)
  ]}
]
variables [
 isSkydivingAlpha {type=number}
 kneeAlpha_l {type=number}
 kneeAlpha_r {type=number}
 LaimPecPos {type=Vector}
 LarmDistanceBICEPAlpha {type=number}
 LbicepFlex {type=number}
 LpecStretchAlpha {type=number}
 LshoulderAlpha {type=number}
 RaimPecPos {type=Vector}
 RarmDistanceAlpha {type=number}
 RarmDistanceBICEPAlpha {type=number}
 RbicepFlex {type=number}
 RflexAlpha {type=number}
 RlatStretchAlpha {type=number}
 RlegAlpha {type=number}
 RpecAimPosAlpha {type=number}
 RpecStretchAlpha {type=number}
 RshoulderAlpha {type=number}
 trapsStretchAlpha_l {type=number}
 trapsStretchAlpha_r {type=number}
 windVector {type=Vector}
]
